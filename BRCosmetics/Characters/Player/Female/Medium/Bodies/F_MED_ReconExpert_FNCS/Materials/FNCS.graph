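# Material function for the FNCS Recon Expert body that layers a thin-film iridescence onto the base color of incoming Material Attributes.
# A Fresnel term, scaled and offset by a parameter or switched to animated modulo time, drives lookups into a spectrum texture.
# The sampled color is added to base color, masked by inverse roughness and a second Fresnel.
# An Alpha input blends the result with the original attributes.
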MaterialFunction /BRCosmetics/Characters/Player/Female/Medium/Bodies/F_MED_ReconExpert_FNCS/Materials/FNCS
graph {nodes [
  n1 = FunctionInput() {inputName=Material Attributes, inputType=FunctionInput_MaterialAttributes}
  n2 = GetMaterialAttributes(None=@n1) {attributeGetTypes [#1, #2]}
  n3 = MaterialFunctionCall!(
   Time
   Divisor=ScalarParameter() {defaultValue=1, group=ThinFilm, parameterName=ThinFilm_Animated_TimeDivisor}
   Mod=ScalarParameter() {defaultValue=1, group=ThinFilm, parameterName=ThinFilm_Animated_TimeModulo}
  ) {materialFunction=/Game/Effects/Fort_Effects/Materials/_Functions/MF_ModuloTime}
  n4 = ScalarParameter() {group=ThinFilm, parameterName=ThinFilm_Offset}
  n5 = ScalarParameter() {defaultValue=1, group=ThinFilm, parameterName=ThinFilm_Scale}
  n6 = Fresnel(ExponentIn=ScalarParameter() {defaultValue=0.5, group=ThinFilm, parameterName=ThinFilm_Warp}) {baseReflectFraction=0}
  n7 = TextureSampleParameter2D(UVs=Add(A=StaticSwitchParameter(True=@n3.Unclamped, False=@n4) {group=ThinFilm, parameterName=AnimateThinFilm}, B=Multiply(A=@n5, B=@n6))) {group=ThinFilm, parameterName=Thin Film Texture, texture=/Game/Global/Textures/Default/Thinfilm/T_ThinFilm_Spectrum_COLOR}
  n8 = ScalarParameter() {defaultValue=0.5, group=ThinFilm, parameterName=ThinFilm_Intensity}
  n9 = Multiply(A=OneMinus(None=@n2.Roughness), B=ScalarParameter() {defaultValue=1, group=ThinFilm, parameterName=ThinFilm_RoughnessScale})
  n10 = ScalarParameter() {defaultValue=1.2164, group=ThinFilm, parameterName=ThinFilm_Exponent}
  n11 = LinearInterpolate(
   A=@n2.BaseColor
   B=Add(A=@n2.BaseColor, B=Multiply(A=@n7.RGB, B=@n8))
   Alpha=Multiply(A=Clamp(None=@n9), B=Fresnel(ExponentIn=@n10) {baseReflectFraction=0})
  )
  FunctionOutput(None=BlendMaterialAttributes(
    A=@n1
    B=SetMaterialAttributes!(MaterialAttributes=@n2.MaterialAttributes, BaseColor=@n11) {attributeSetTypes [#2]}
    Alpha=FunctionInput() {bUsePreviewValueAsDefault=true, inputName=Alpha, inputType=FunctionInput_Scalar}
   ) {vertexAttributeBlendType=UseA})
 ]}
