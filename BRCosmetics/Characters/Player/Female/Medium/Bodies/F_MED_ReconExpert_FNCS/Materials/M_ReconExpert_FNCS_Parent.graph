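# Parent material for the female ReconExpert FNCS character body, a Default Lit skeletal-mesh surface built on material attributes.
# The shared character PreFX function feeds an FNCS effect function, then a metals LUT function, then the shared PostFX function.
# Selectable channels of one SkinFX Mask texture, black by default, mask the FNCS and metal stages.
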Material /BRCosmetics/Characters/Player/Female/Medium/Bodies/F_MED_ReconExpert_FNCS/Materials/M_ReconExpert_FNCS_Parent
graph {
 properties {
  bUseMaterialAttributes=true
  bUsedWithMeshDeformer=true
  bUsedWithMorphTargets=true
  bUsedWithSkeletalMesh=true
  pixelDepthOffsetMode=PDOM_Legacy
  usedShadingModels=Default Lit
 }
 nodes [
  n1 = MaterialFunctionCall() {materialFunction=&BaseFN_CharacterFunction_PreFX}
  n2 = TextureSampleParameter2D() {parameterName=SkinFX Mask, samplerType=SAMPLERTYPE_Masks, texture=&T_BlackMask}
  n3 = ChannelMaskParameter(None=@n2.RGB) {parameterName=FNCS Mask}
  MaterialFunctionCall(
   AO=@n1.AO
   Distance Blend=@n1.AO
   Fresnel=@n1.AO
   Gmap=@n1.AO
   Pre Skinned Local Position=@n1.AO
   Material=MaterialFunctionCall(
    Mask=ChannelMaskParameter(None=@n2.RGB) {parameterName=Metal Mask}.R
    Materials=MaterialFunctionCall!(Alpha=@n3.R, Material Attributes=@n1.Result) {materialFunction=&/FNCS}.Result
   ) {materialFunction=&MF_MetalsLUT}.Result
  ) {materialFunction=&BaseFN_CharacterFunction_PostFX}
 ]
}
